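# MF_AnimalEyes_TextureInputs is an empty material function in the AnimalEyes_Parent folder of the 2023 character eye parent materials.
# Its expression graph holds no nodes, so it declares no inputs, outputs or texture samples.
# The name suggests it is meant to hold texture inputs for animal eye materials, but no such logic is present.
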MaterialFunction /Game/Global/Materials/Character_Parent_2023/Eye_Parent_2023/AnimalEyes_Parent/MF/MF_AnimalEyes_TextureInputs
graph {expressions {}}
