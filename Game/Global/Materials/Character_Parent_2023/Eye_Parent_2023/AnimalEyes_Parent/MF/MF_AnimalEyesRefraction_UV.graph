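# Material function for animal eyes that outputs UVs offset to fake corneal refraction over the iris.
# A Snell's law calculation bends the camera vector at the pixel normal using an air-to-eye IoR ratio.
# A displacement map, measured relative to the iris edge, gives depth, scaled by facing against a sphere normal map's eye direction.
# The refraction vector is projected onto an orthogonalized tangent basis and added to the UVs.
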MaterialFunction /Game/Global/Materials/Character_Parent_2023/Eye_Parent_2023/AnimalEyes_Parent/MF/MF_AnimalEyesRefraction_UV
graph {nodes [
  n1 = NamedRerouteUsage()
  n2 = MaterialFunctionCall() {materialFunction=&TangentBasis}
  n3 = NamedRerouteUsage()
  n4 = DotProduct(A=@n2.X U, B=@n3)
  n5 = Normalize(VectorInput=Subtract(A=@n2.X U, B=Multiply(A=@n4, B=@n3)))
  n6 = CrossProduct(A=@n5, B=@n3)
  n7 = Multiply(
   A=Multiply(A=ScalarParameter() {defaultValue=0.02, group=EyeRefraction_UV, parameterName=Eye Refraction Mult}, B=Constant2Vector() {g=1, r=-1})
   B=MaterialFunctionCall(X=DotProduct(A=@n5, B=@n1), Y=DotProduct(A=@n6, B=@n1)) {materialFunction=&MakeFloat2}.Result
  )
  n8 = TextureCoordinate()
  FunctionOutput(None=NamedRerouteDeclaration(None=Add(A=@n7, B=@n8)) {name=Refracted UVs, nodeColor=<rgba 0.6406 1 0 1>}) {outputName=EyeRefraction_UV}
  n9 = FunctionInput(Preview=TextureCoordinate()) {bUsePreviewValueAsDefault=true, inputName=In UV Position, inputType=FunctionInput_Vector2}
  n10 = TextureSample(UVs=@n9, Tex=TextureObject() {samplerType=SAMPLERTYPE_Normal, texture=/Game/Characters/Shared/FNEye/Textures/T_Eye_Sphere_N}) {samplerSource=SSM_Clamp_WorldGroupSettings, samplerType=SAMPLERTYPE_Normal, texture=&T_Eye_Sphere_N}
  NamedRerouteDeclaration(None=Transform(None=@n10.RGB)) {name=EyeDirection, nodeColor=<rgba 1 0.3047 0 1>}
  NamedRerouteDeclaration(None=CameraVectorWS()) {name=Camera Vector, nodeColor=<rgba 1 0 1 1>}
  NamedRerouteDeclaration(None=PixelNormalWS()) {name=_Normal, nodeColor=<rgba 0 0 1 1>}
  NamedRerouteDeclaration(None=Divide(A=Constant() {desc=airIoR, r=1.0003}, B=Constant() {r=1.336}) {constA=1.0003}) {name=n, nodeColor=<rgba 0 1 0.2031 1>}
  n11 = NamedRerouteUsage()
  n12 = NamedRerouteUsage()
  NamedRerouteDeclaration(None=Multiply(A=NamedRerouteUsage(), B=DotProduct(A=@n11, B=@n12) {desc=facing}) {desc=w}) {name=w, nodeColor=<rgba 1 0.8297 0 1>}
  n13 = NamedRerouteUsage()
  n14 = NamedRerouteUsage()
  n15 = Multiply(A=Subtract(A=@n13, B=@n14), B=Add(A=@n13, B=@n14))
  NamedRerouteDeclaration(None=SquareRoot(None=Add(A=@n15)) {desc=k}) {name=k, nodeColor=<rgba 0.518 1 0 1>}
  n16 = NamedRerouteUsage()
  n17 = NamedRerouteUsage()
  n18 = Subtract(A=Multiply(A=Subtract(A=@n16, B=@n17), B=NamedRerouteUsage()), B=Multiply(A=NamedRerouteUsage(), B=NamedRerouteUsage())) {desc=t}
  NamedRerouteDeclaration(None=Multiply(A=Normalize(VectorInput=@n18)) {constB=-1}) {name=Eye Refraction Direction, nodeColor=<rgba 0 0.2882 1 1>}
  n19 = TextureObjectParameter() {
   group=EyeRefraction_UV
   parameterName=EyeRefraction_Displacement
   samplerType=SAMPLERTYPE_LinearColor
   texture=/Game/Characters/Shared/FNEye/Textures/T_EyeMidPlaneDisplacement
  }
  n20 = TextureSample(UVs=@n9, Tex=@n19) {samplerType=SAMPLERTYPE_Masks}
  n21 = ScalarParameter() {defaultValue=0.4, group=EyeRefraction_UV, parameterName=Eye Iris UV Radius, sliderMax=0.5, sortPriority=107}
  n22 = TextureSample(
   UVs=MaterialFunctionCall(X=Add(A=@n21) {constB=0.5}, Y=Constant() {r=0.5}) {materialFunction=&MakeFloat2}.Result
   Tex=@n19
  ) {
   desc="Sample Displacement map as above, sampled at the Iris Radius"
   samplerType=SAMPLERTYPE_LinearColor
   texture=&T_EyeMidPlaneDisplacement
  }
  n23 = Multiply(
   A=Max(A=Subtract(A=@n20.R, B=@n22.R)) {constB=0}
   B=ScalarParameter() {defaultValue=0.2, group=EyeRefraction_UV, parameterName=EyeRefraction_ DepthScale}
  )
  n24 = DotProduct(A=CameraVectorWS(), B=NamedRerouteUsage())
  n25 = LinearInterpolate(Alpha=Multiply(A=@n24, B=@n24)) {constA=0.325}
  NamedRerouteDeclaration(None=Multiply(A=NamedRerouteUsage(), B=Divide(A=@n23, B=@n25) {desc=heightW / cosAlpha - scale distance to iris plane by facing ratio of camera and plane})) {name=Refraction Vector, nodeColor=<rgba 0.0781 1 0 1>}
  TextureCoordinate()
  TextureCoordinate()
 ]}
