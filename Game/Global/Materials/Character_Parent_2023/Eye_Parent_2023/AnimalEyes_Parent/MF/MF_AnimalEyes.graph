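# Material function that shades animal eyes, layering sclera, tintable iris and pupil over incoming material attributes, with separate mask and UV outputs.
# Refraction-adjusted, center-scaled UVs sample the sclera, while iris UVs get a nonlinear radial rescale and an optional two-color remap.
# The pupil is procedural, using a sum-of-powers ellipse with adjustable size, shape and contrast.
# Masks drive color, roughness, specular, metallic and a color-coded emissive debug view.
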MaterialFunction /Game/Global/Materials/Character_Parent_2023/Eye_Parent_2023/AnimalEyes_Parent/MF/MF_AnimalEyes
graph {nodes [
  FunctionInput() {bUsePreviewValueAsDefault=true, inputName=------------------------------------_1, inputType=FunctionInput_Scalar, sortPriority=11}
  FunctionInput() {bUsePreviewValueAsDefault=true, inputName=------------------------------------_2, inputType=FunctionInput_Scalar, sortPriority=13}
  FunctionInput() {bUsePreviewValueAsDefault=true, inputName=------------------------------------_3, inputType=FunctionInput_Scalar, sortPriority=16}
  n1 = GetMaterialAttributes(None=FunctionInput() {bUsePreviewValueAsDefault=true, inputName=InputMaterial, inputType=FunctionInput_MaterialAttributes, sortPriority=6}) {attributeGetTypes [#1]}
  n2 = SetMaterialAttributes!(
   MaterialAttributes=@n1.MaterialAttributes
   BaseColor=LinearInterpolate(A=NamedRerouteUsage(), Alpha=NamedRerouteUsage()) {constB=0}
   Metallic=NamedRerouteUsage()
   Specular=NamedRerouteUsage()
   Roughness=NamedRerouteUsage()
   EmissiveColor=NamedRerouteUsage()
   Normal=NamedRerouteUsage()
   ShadingModel=ShadingModel()
  ) {attributeSetTypes [
    #2
    #3
    #4
    #5
    #6
    #1
    #7
   ]}
  FunctionOutput(None=BlendMaterialAttributes(A=@n1.MaterialAttributes, B=@n2, Alpha=NamedRerouteUsage()) {vertexAttributeBlendType=UseA}) {sortPriority=3}
  FunctionOutput(None=NamedRerouteUsage()) {outputName=IrisMask, sortPriority=1}
  n3 = Abs(None=NamedRerouteUsage())
  n4 = NamedRerouteUsage()
  n5 = LinearInterpolate(Alpha=@n4) {constA=0.03}
  n6 = LinearInterpolate(A=Constant() {r=0.8}, Alpha=@n4) {constB=2}
  n7 = Abs(None=NamedRerouteUsage())
  n8 = LinearInterpolate(A=NamedRerouteUsage(), Alpha=Power(Base=NamedRerouteUsage()))
  n9 = Add(A=Power(Base=Divide(A=@n3, B=@n5), Exp=@n6), B=Power(Base=Divide(A=@n7, B=@n8), Exp=@n6))
  n10 = MaterialFunctionCall(In=OneMinus(None=@n9), Contrast=NamedRerouteUsage()) {materialFunction=&CheapContrast}
  FunctionOutput(None=NamedRerouteDeclaration(None=Saturate(None=@n10.Result)) {name=Eye_pupil_Mask, nodeColor=<rgba 0.5938 1 0 1>}) {outputName=PupilMask, sortPriority=2}
  n11 = NamedRerouteUsage()
  n12 = Constant() {r=0.5}
  n13 = NamedRerouteUsage()
  n14 = NamedRerouteUsage()
  FunctionOutput(None=NamedRerouteDeclaration(None=SphereMask(A=@n11, B=@n12, Radius=@n13, Hardness=@n14) {hardnessPercent=90}) {name=ScleraMask, nodeColor=<rgba 0.1484 1 0 1>}) {outputName=Sclera_Mask}
  n15 = MaterialFunctionCall(UVs=Add(A=TextureCoordinate(), B=NamedRerouteUsage()), Texture Scale=NamedRerouteUsage()) {materialFunction=&ScaleUVsByCenter}
  FunctionOutput(None=NamedRerouteDeclaration(None=@n15.UVs) {name=Eye_Sclera_UV_1, nodeColor=<rgba 0.2422 1 0 1>}) {outputName=Eye_Sclera_UV, sortPriority=3}
  FunctionOutput(None=NamedRerouteUsage()) {outputName=BaseColor, sortPriority=1}
  NamedRerouteDeclaration(None=FunctionInput() {bUsePreviewValueAsDefault=true, inputName=Eye_Sclera_UV_Position, inputType=FunctionInput_Vector2, sortPriority=10}) {name=AnimalEyes_Sclera_UV_Offset, nodeColor=<rgba 1 0.3984 0 1>}
  n16 = FunctionInput() {
   bUsePreviewValueAsDefault=true
   inputName=Eye_Sclera_UV_Scale_Offset
   inputType=FunctionInput_Vector2
   sortPriority=10
   previewValue=<0.73 0.73 0 1>
  }
  NamedRerouteDeclaration(None=@n16) {name=AnimalEyes_Sclera_UV_Scale_Offset, nodeColor=<rgba 0.0781 0 1 1>}
  n17 = FunctionInput() {
   bUsePreviewValueAsDefault=true
   inputName=Eye_Iris_UV_Scale_Offset
   inputType=FunctionInput_Vector2
   sortPriority=12
   previewValue=<0.33 0.33 0 1>
  }
  NamedRerouteDeclaration(None=@n17) {name=Eye_Iris_UV_Scale_Offset, nodeColor=<rgba 0.1016 1 0 1>}
  n18 = FunctionInput() {bUsePreviewValueAsDefault=true, inputName=pupil_size, inputType=FunctionInput_Scalar, sortPriority=25, previewValue=<0.544 0 0 1>}
  NamedRerouteDeclaration(None=@n18) {name=Pupil size, nodeColor=<rgba 0 1 0.5078 1>}
  n19 = FunctionInput() {bUsePreviewValueAsDefault=true, inputName=pupil_shape, inputType=FunctionInput_Scalar, sortPriority=25}
  NamedRerouteDeclaration(None=Saturate(None=OneMinus(None=@n19))) {name=Pupil Dilation, nodeColor=<rgba 0 0.5312 1 1>}
  n20 = TextureObjectParameter() {
   group=AnimalEyes_06Texture
   parameterName=T_Eye_Sphere_N
   samplerType=SAMPLERTYPE_Normal
   texture=&FlatNormalBlue
  }
  NamedRerouteDeclaration(None=@n20) {name=T_Eye_Sphere_N, nodeColor=<rgba 0.1484 1 0 1>}
  NamedRerouteDeclaration(None=@n1.Normal) {name=IN_Normal, nodeColor=<rgba 0 0.6953 1 1>}
  NamedRerouteDeclaration(None=TextureObjectParameter() {group=AnimalEyes_06Texture, parameterName=T_EyeSclera_D, texture=/Game/Global/Textures/Default/Blanks/T_MidgreysRGB}) {name=T_EyeSclera_D, nodeColor=<rgba 0.1484 1 0 1>}
  NamedRerouteDeclaration(None=TextureObjectParameter() {group=AnimalEyes_06Texture, parameterName=T_Eye_Iris, texture=&T_MidgreysRGB}) {name=T_Eye_Iris, nodeColor=<rgba 0.1484 1 0 1>}
  n21 = TextureObjectParameter() {
   group=AnimalEyes_06Texture
   parameterName=T_EyeLirs2_S
   samplerType=SAMPLERTYPE_Masks
   texture=&T_Fortnite_Default_S
  }
  NamedRerouteDeclaration(None=@n21) {name=T_EyeLirs2_S, nodeColor=<rgba 0.1484 1 0 1>}
  NamedRerouteDeclaration(None=@n9) {name=Eye_pupil_UV, nodeColor=<rgba 0 0.6953 1 1>}
  n22 = FunctionInput() {bUsePreviewValueAsDefault=true, inputName=Eye_Iris_Mask_Radius, inputType=FunctionInput_Scalar, sortPriority=12, previewValue=<0.485 0 0 1>}
  NamedRerouteDeclaration(None=@n22) {name=Eye_Iris_Mask_Radius, nodeColor=<rgba 1 0 0.8672 1>}
  NamedRerouteDeclaration(None=MaterialFunctionCall(UVs=@n15.UVs, Texture Scale=NamedRerouteUsage()) {materialFunction=&ScaleUVsByCenter}.UVs) {name=Eye_Iris_UV_1, nodeColor=<rgba 1 0.9844 0 1>}
  n23 = TextureSample(UVs=NamedRerouteUsage(), Tex=NamedRerouteUsage()) {samplerType=SAMPLERTYPE_Masks}
  n24 = NamedRerouteUsage()
  NamedRerouteDeclaration(None=LinearInterpolate(A=LinearInterpolate(B=@n23.G, Alpha=@n24) {constA=1}, Alpha=NamedRerouteUsage()) {constB=0.2}) {name=EyeSpecular, nodeColor=<rgba 0.8516 1 0 1>}
  n25 = Subtract(A=NamedRerouteUsage(), B=NamedRerouteUsage())
  NamedRerouteDeclaration(None=LinearInterpolate(B=Constant() {r=0.75}, Alpha=Saturate(None=@n25)) {constB=0.8}) {name=EyeMetallic, nodeColor=<rgba 0.8281 0 1 1>}
  n26 = Constant() {r=0.1}
  n27 = SphereMask(A=NamedRerouteUsage(), B=Constant() {r=0.5}, Radius=NamedRerouteUsage(), Hardness=NamedRerouteUsage()) {hardnessPercent=90}
  n28 = NamedRerouteDeclaration(None=@n27) {name=Lris_Mask, nodeColor=<rgba 1 0 0.6562 1>}
  NamedRerouteDeclaration(None=LinearInterpolate(A=LinearInterpolate(A=@n26, B=@n23.R, Alpha=@n28), B=Constant() {r=0.15}, Alpha=NamedRerouteUsage())) {name=EyeRoughness, nodeColor=<rgba 0 1 0.5078 1>}
  n29 = FunctionInput() {bUsePreviewValueAsDefault=true, inputName=Eye_Iris_Mask_Hardness, inputType=FunctionInput_Scalar, sortPriority=12, previewValue=<0.93 0 0 1>}
  NamedRerouteDeclaration(None=@n29) {name=Eye_Iris_Mask_Hardness, nodeColor=<rgba 0 1 0.8125 1>}
  n30 = FunctionInput() {bUsePreviewValueAsDefault=true, inputName=IrisEdgeBrightness, inputType=FunctionInput_Scalar, sortPriority=12, previewValue=<1 0 0 1>}
  n31 = LinearInterpolate(A=NamedRerouteUsage(), B=Multiply(A=NamedRerouteUsage(), B=@n30), Alpha=NamedRerouteUsage()) {constB=0}
  n32 = NamedRerouteUsage()
  n33 = NamedRerouteUsage()
  n34 = FunctionInput() {bUsePreviewValueAsDefault=true, inputName=Eye_Pupil_ColorTint, sortPriority=15, previewValue=<0.002 0.0019 0.001 1>}
  NamedRerouteDeclaration(None=LinearInterpolate(A=LinearInterpolate(A=@n31, B=@n32, Alpha=@n33), B=@n34, Alpha=NamedRerouteUsage()) {constB=0}) {name=BaseColor, nodeColor=<rgba 0.2188 0 1 1>}
  NamedRerouteDeclaration(None=Constant() {r=30}) {name=Pupil_Size_Max, nodeColor=<rgba 1 0.4922 0 1>}
  n35 = FunctionInput() {bUsePreviewValueAsDefault=true, inputName=AnimalEyes_Sclera_Color_Tint, sortPriority=10, previewValue=<1 1 1 1>}
  NamedRerouteDeclaration(None=@n35) {name=AnimalEyes_Sclera_Color, nodeColor=<rgba 1 0.75 0 1>}
  NamedRerouteDeclaration(None=Constant() {r=6}) {name=Pupil_Size_Min, nodeColor=<rgba 0.1719 0 1 1>}
  n36 = TextureSample(UVs=NamedRerouteUsage(), Tex=NamedRerouteUsage()) {samplerSource=SSM_Clamp_WorldGroupSettings, samplerType=SAMPLERTYPE_Normal, texture=/Game/Characters/Shared/FNEye/Textures/T_Eye_Sphere_N}
  NamedRerouteDeclaration(None=@n36.RGB) {name=OutputNormal, nodeColor=<rgba 0.1016 1 0 1>}
  NamedRerouteDeclaration(None=Constant() {r=1.5}) {name=Pupil_VerticalDilation, nodeColor=<rgba 0 1 0.7188 1>}
  n37 = FunctionInput() {bUsePreviewValueAsDefault=true, inputName=Eye_Pupil_Contrast, inputType=FunctionInput_Scalar, sortPriority=15, previewValue=<3 0 0 1>}
  NamedRerouteDeclaration(None=@n37) {name=Pupil_Contrast, nodeColor=<rgba 1 0.3984 0 1>}
  n38 = FunctionInput() {
   bUsePreviewValueAsDefault=true
   inputName=Eye_Sclera_Mask_Radius
   inputType=FunctionInput_Scalar
   sortPriority=10
   previewValue=<0.485 0 0 1>
  }
  NamedRerouteDeclaration(None=@n38) {name=AnimalEyes_Sclera_Mask_Radius, nodeColor=<rgba 1 0 0.8437 1>}
  n39 = FunctionInput() {
   bUsePreviewValueAsDefault=true
   inputName=Eye_Sclera_Mask_Hardness
   inputType=FunctionInput_Scalar
   sortPriority=10
   previewValue=<0.93 0 0 1>
  }
  NamedRerouteDeclaration(None=@n39) {name=AnimalEyes_Sclera_Mask_Hardness, nodeColor=<rgba 0 1 0.8359 1>}
  n40 = NamedRerouteUsage()
  n41 = NamedRerouteUsage()
  n42 = NamedRerouteUsage()
  n43 = Constant() {r=0.389}
  n44 = NamedRerouteUsage()
  MaterialExpressionMultiply_13 = Multiply(
   A=Add(A=NamedRerouteUsage()) {constB=-0.5}
   B=Multiply(A=LinearInterpolate(A=@n40, B=@n41, Alpha=@n42) {constA=1, constB=0}, B=LinearInterpolate(B=@n43, Alpha=@n44) {constA=1})
  ) {constB=2}
  n45 = Convert(None=@MaterialExpressionMultiply_13) {
   nodeName=Break Float2
   convertInputs [{
     type=Vector2
     expressionInput {
      expression=&self:MaterialExpressionMultiply_13
      inputName=None
      mask=0
      maskA=0
      maskB=0
      maskG=0
      maskR=0
      outputIndex=0
     }
    }]
   convertOutputs [{type=Scalar}, {type=Scalar}]
  }
  NamedRerouteDeclaration(None=@n45) {name=L_pupil_UVx, nodeColor=<rgba 0.1719 1 0 1>}
  NamedRerouteDeclaration(None=FunctionInput() {bUsePreviewValueAsDefault=true, inputName=Eye_Debug, inputType=FunctionInput_Scalar, sortPriority=21}) {name=Eye_Debug, nodeColor=<rgba 0 0.5312 1 1>}
  n46 = LinearInterpolate(
   A=LinearInterpolate(A=Constant3Vector() {constant=<rgba 1 1 1 0>}, B=Constant3Vector() {constant=<rgba 1 0 0 0>}, Alpha=NamedRerouteUsage())
   B=Constant4Vector() {constant=<rgba 0 1 0 1>}.RGB
   Alpha=NamedRerouteUsage()
  )
  n47 = Constant3Vector() {constant=<rgba 0 0 1 0>}
  n48 = NamedRerouteUsage()
  NamedRerouteDeclaration(None=LinearInterpolate(A=NamedRerouteUsage(), B=LinearInterpolate(A=@n46, B=@n47, Alpha=@n48), Alpha=NamedRerouteUsage())) {name=EmissiveColor, nodeColor=<rgba 1 0 0.3281 1>}
  NamedRerouteDeclaration(None=Saturate(None=Multiply(A=@n27) {constB=10})) {name=Lris_Mask_EdgeHard, nodeColor=<rgba 1 0 0.6562 1>}
  NamedRerouteDeclaration(None=@n45) {name=L_pupil_UVy, nodeColor=<rgba 0.1719 1 0 1>}
  n49 = MaterialFunctionCall(UVs=NamedRerouteUsage(), Texture Scale=NamedRerouteUsage()) {materialFunction=&ScaleUVsByCenter}
  n50 = Power(Base=NamedRerouteUsage()) {constExponent=2.5}
  n51 = NamedRerouteUsage()
  n52 = Multiply(A=NamedRerouteUsage(), B=LinearInterpolate(A=Constant() {r=0.22}, Alpha=Power(Base=@n51) {constExponent=4}) {constA=1}) {constB=0.5}
  n53 = Power(Base=LinearInterpolate(Alpha=Length(None=@n52) {is=class MaterialExpressionLength}) {constB=0.4}, Exp=Constant() {r=0.5}) {constExponent=3}
  n54 = LinearInterpolate(
   A=LinearInterpolate(A=Constant() {r=0.524}, B=Constant() {r=3}, Alpha=Saturate(None=@n50))
   B=Constant() {r=1}
   Alpha=Saturate(None=Saturate(None=@n53))
  ) {constA=1}
  n55 = Constant() {r=0.5}
  n56 = Add(A=Divide(A=NamedRerouteDeclaration(None=@n49.UVs) {name=Eye_Iris_UV, nodeColor=<rgba 1 0.9844 0 1>}, B=@n54), B=@n55)
  n57 = Divide(A=@n55, B=@n54) {desc=Scale the center offset by the scaling factor}
  n58 = TextureSample(
   UVs=NamedRerouteDeclaration(None=Subtract(A=@n56, B=@n57) {desc=Re offset to maintain center}) {name=Eye_Iris_anim_UV, nodeColor=<rgba 1 0 0.2109 1>}
   Tex=NamedRerouteUsage()
  ) {samplerSource=SSM_Clamp_WorldGroupSettings}
  n59 = FunctionInput() {bUsePreviewValueAsDefault=true, inputName=Iris_Color_Primary, sortPriority=12, previewValue=<1 1 1 1>}
  n60 = FunctionInput() {bUsePreviewValueAsDefault=true, inputName=Iris_Color_Secondary, sortPriority=12, previewValue=<1 1 1 1>}
  n61 = Constant() {r=1}
  n62 = FunctionInput() {bUsePreviewValueAsDefault=true, inputName=lris_ColorContrast, inputType=FunctionInput_Scalar, sortPriority=12, previewValue=<0 1 1 1>}
  n63 = LinearInterpolate(
   A=@n59
   B=@n60
   Alpha=MaterialFunctionCall(In=Desaturation(None=@n58.RGB, Fraction=@n61), Contrast=@n62) {materialFunction=&CheapContrast}.Result
  )
  n64 = FunctionInput() {bUsePreviewValueAsDefault=true, inputName=lris_ColorOffset_Enable, inputType=FunctionInput_Scalar, sortPriority=12, previewValue=<0 1 1 1>}
  NamedRerouteDeclaration(None=LinearInterpolate(A=@n58.RGB, B=Saturate(None=@n63), Alpha=@n64)) {name=Eye_Iris, nodeColor=<rgba 1 0 0.1875 1>}
  n65 = MaterialFunctionCall(UVs=Add(A=TextureCoordinate(), B=NamedRerouteUsage()), Texture Scale=NamedRerouteUsage()) {materialFunction=&ScaleUVsByCenter}
  n66 = NamedRerouteUsage()
  n67 = FunctionInput() {bUsePreviewValueAsDefault=true, inputName=Eye Refraction Mix, inputType=FunctionInput_Scalar, sortPriority=21, previewValue=<1 0 0 1>}
  n68 = LinearInterpolate(
   A=TextureCoordinate()
   B=MaterialFunctionCall!(In UV Position=MaterialFunctionCall(UVs=@n65.UVs, Texture Scale=@n66) {materialFunction=&ScaleUVsByCenter}.UVs) {materialFunction=&/MF_AnimalEyesRefraction_UV}.EyeRefraction_UV
   Alpha=@n67
  )
  n69 = NamedRerouteUsage()
  n70 = NamedRerouteDeclaration(None=MaterialFunctionCall(UVs=Add(A=@n68, B=@n69), Texture Scale=NamedRerouteUsage()) {materialFunction=&ScaleUVsByCenter}.UVs) {name=Eye_Sclera_UV, nodeColor=<rgba 0.2422 1 0 1>}
  n71 = NamedRerouteUsage()
  NamedRerouteDeclaration(None=Multiply(A=TextureSample(UVs=@n70, Tex=@n71) {samplerSource=SSM_Clamp_WorldGroupSettings}.RGB, B=NamedRerouteUsage())) {name=Eye_Sclera, nodeColor=<rgba 0 0.5078 1 1>}
  NamedRerouteUsage()
  NamedRerouteUsage()
 ]}
