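# MFW_AnimalEyes is a material function wrapper that exposes MF_AnimalEyes as grouped parameters for animal eye materials.
# It feeds sclera, iris, pupil, refraction and debug parameters into that function, outputting material attributes, base color, layer masks and sclera UV.
# Static switches select uniform or per-axis sclera and iris scale.
# An optional dual-eye switch blends left and right sclera positions with a hard step on the V coordinate.
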MaterialFunction /Game/Global/Materials/Character_Parent_2023/Eye_Parent_2023/AnimalEyes_Parent/MF/MFW_AnimalEyes
graph {nodes [
  n1 = VectorParameter() {group=AnimalEyes_02Sclera, parameterName=AnimalEyes_Sclera_Color_Primary, sortPriority=52, defaultValue=<rgba 1 1 1 1>}
  n2 = VectorParameter() {group=AnimalEyes_03lris, parameterName=AnimalEyes_lris_Color_Secondary, sortPriority=53, defaultValue=<rgba 1 1 1 1>}
  n3 = VectorParameter() {group=AnimalEyes_04Pupils, parameterName=AnimalEyes_Pupil_Color_Primary, sortPriority=54, defaultValue=<rgba 0.002 0.0019 0.002 1>}
  n4 = VectorParameter() {group=AnimalEyes_03lris, parameterName=AnimalEyes_lris_Color_Primary, sortPriority=53, defaultValue=<rgba 1 1 1 1>}
  n5 = MaterialFunctionCall!(
   InputMaterial=FunctionInput() {bUsePreviewValueAsDefault=true, inputName=IN_Material, inputType=FunctionInput_MaterialAttributes, sortPriority=6}
   Eye_Sclera_UV_Scale_Offset=NamedRerouteUsage()
   Eye_Sclera_UV_Position=NamedRerouteUsage()
   Eye_Sclera_Mask_Hardness=ScalarParameter() {defaultValue=0.93, group=AnimalEyes_02Sclera, parameterName=AnimalEyes_Sclera_Mask_Hardness, sortPriority=52}
   Eye_Sclera_Mask_Radius=ScalarParameter() {defaultValue=0.485, group=AnimalEyes_02Sclera, parameterName=AnimalEyes_Sclera_Mask_Radius, sortPriority=52}
   AnimalEyes_Sclera_Color_Tint=@n1.RGB
   ------------------------------------_1
   Iris_Color_Secondary=@n2.RGB
   Eye_Iris_Mask_Hardness=ScalarParameter() {defaultValue=0.93, group=AnimalEyes_03lris, parameterName=AnimalEyes_Iris_Mask_Hardness, sortPriority=53}
   IrisEdgeBrightness=ScalarParameter() {defaultValue=1, group=AnimalEyes_03lris, parameterName=AnimalEyes_Iris_EdgeBrightness, sortPriority=53}
   Eye_Iris_UV_Scale_Offset=NamedRerouteUsage()
   Eye_Iris_Mask_Radius=ScalarParameter() {defaultValue=0.4, group=AnimalEyes_03lris, parameterName=AnimalEyes_Iris_Mask_Radius, sortPriority=53}
   lris_ColorOffset_Enable=ScalarParameter() {group=AnimalEyes_03lris, parameterName=AnimalEyes_Use_lris_Color_Offset, sortPriority=53}
   Iris_Color_Primary=@n4.RGB
   lris_ColorContrast=ScalarParameter() {group=AnimalEyes_03lris, parameterName=AnimalEyes_lrisColor_Contrast, sortPriority=53}
   ------------------------------------_2
   Eye_Pupil_Contrast=ScalarParameter() {defaultValue=3, group=AnimalEyes_04Pupils, parameterName=AnimalEyes_Pupil_Edge_Softness, sortPriority=54}
   Eye_Pupil_ColorTint=@n3.RGB
   ------------------------------------_3
   Eye Refraction Mix=ScalarParameter() {defaultValue=1, group=AnimalEyes_01Advance, parameterName=AnimalEyes_Refraction Mix, sortPriority=51}
   Eye_Debug=ScalarParameter() {group=AnimalEyes_01Advance, parameterName=AnimalEyes_Debug_RGB, sortPriority=51}
   pupil_shape=ScalarParameter() {group=AnimalEyes_04Pupils, parameterName=pupil_shape, sliderMax=1, sortPriority=54}
   pupil_size=ScalarParameter() {defaultValue=0.5448, group=AnimalEyes_04Pupils, parameterName=pupil_size, sliderMax=1, sortPriority=54}
  ) {materialFunction=&/MF_AnimalEyes}
  FunctionOutput(None=@n5.Result) {outputName=Out_MA}
  FunctionOutput(None=@n5.Sclera_Mask) {outputName=Out_Sclera_Mask}
  FunctionOutput(None=@n5.IrisMask) {outputName=Out_lris_Mask}
  FunctionOutput(None=@n5.BaseColor) {outputName=Out_BaseColor}
  FunctionOutput(None=@n5.PupilMask) {outputName=Out_Pupil_Mask}
  FunctionOutput(None=@n5.Eye_Sclera_UV) {outputName=Eye_Sclera_UV}
  n6 = VectorParameter() {group=AnimalEyes_02Sclera, parameterName=AnimalEyes_Sclera_UV_Position_L, defaultValue=<rgba -0.413 0.415 0 1>}
  n7 = ComponentMask(None=@n6.RGB) {g=true, r=true}
  n8 = VectorParameter() {group=AnimalEyes_02Sclera, parameterName=AnimalEyes_Sclera_UV_Position_R, defaultValue=<rgba -0.413 0.25 0 1>}
  n9 = ComponentMask(None=@n8.RGB) {g=true, r=true}
  n10 = Constant() {r=0.2}
  n11 = TextureCoordinate()
  n12 = NamedRerouteDeclaration(None=SmoothStep(Min=@n10, Max=@n10, Value=ComponentMask(None=@n11) {g=true})) {name=L_R_Eye_mask, nodeColor=<rgba 1 0 0.7734 1>}
  NamedRerouteDeclaration(None=StaticSwitchParameter(True=LinearInterpolate(A=@n7, B=@n9, Alpha=@n12), False=@n7) {group=AnimalEyes_02Sclera, parameterName=AnimalEyes_Use_DualEye, sortPriority=51}) {name=Eye_Sclera_UV_Position, nodeColor=<rgba 0.6172 1 0 1>}
  n13 = VectorParameter() {group=AnimalEyes_02Sclera, parameterName=AnimalEyes_Scale_UV, sortPriority=52, defaultValue=<rgba 1 1 0 1>}
  NamedRerouteDeclaration(None=StaticSwitchParameter(
    True=ScalarParameter() {defaultValue=0.73, group=AnimalEyes_02Sclera, parameterName=AnimalEyes_ScalerUV_Scale_Uniform, sortPriority=52}
    False=ComponentMask(None=@n13.RGB) {g=true, r=true}
   ) {defaultValue=true, group=AnimalEyes_02Sclera, parameterName=AnimalEyes_ScleraUV_Scale_Uniform_Enable, sortPriority=51}) {name=Eye_Sclera_UV_Scale_Offset_In, nodeColor=<rgba 0 1 0.7422 1>}
  n14 = VectorParameter() {group=AnimalEyes_03lris, parameterName=AnimalEyes_Iris_UV_Scale, defaultValue=<rgba 0 0 0 1>}
  NamedRerouteDeclaration(None=StaticSwitchParameter(
    True=ScalarParameter() {defaultValue=0.33, group=AnimalEyes_03lris, parameterName=AnimalEyes_Iris_UV_Scale_Uniform, sortPriority=53}
    False=ComponentMask(None=@n14.RGB) {g=true, r=true}
   ) {defaultValue=true, group=AnimalEyes_03lris, parameterName=AnimalEyes_IrisUV_Scale_Uniform_Enable, sortPriority=51}) {name=Eye_Iris_UV_Scale_Offset_In, nodeColor=<rgba 1 0.4219 0 1>}
 ]}
