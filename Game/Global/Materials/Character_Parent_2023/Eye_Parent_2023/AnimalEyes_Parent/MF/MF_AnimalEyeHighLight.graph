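# Material function that paints a small circular highlight onto animal eye materials.
# A sphere mask on the input UV, with parameter-driven position and radius, forms a hard-edged dot.
# The dot is added to incoming BaseColor, scaled by UseAnimalEyesHighlight, and to EmissiveColor, scaled independently by a strength parameter that is zero by default.
# It outputs the modified material attributes plus the raw mask as AnimalHighlightMask.
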MaterialFunction /Game/Global/Materials/Character_Parent_2023/Eye_Parent_2023/AnimalEyes_Parent/MF/MF_AnimalEyeHighLight
graph {nodes [
  n1 = GetMaterialAttributes(None=FunctionInput() {bUsePreviewValueAsDefault=true, inputType=FunctionInput_MaterialAttributes}) {attributeGetTypes [#1, #2]}
  n2 = VectorParameter() {group=AnimalEyes_Highlight, parameterName=AnimalEyes_Highlight_POS, defaultValue=<rgba 0.414 0.454 0 1>}
  n3 = SphereMask(
   A=FunctionInput(Preview=TextureCoordinate()) {bUsePreviewValueAsDefault=true, inputName=In_UV, inputType=FunctionInput_Vector2}
   B=ComponentMask(None=@n2.RGB) {g=true, r=true}
   Radius=ScalarParameter() {defaultValue=0.02, group=AnimalEyes_Highlight, parameterName=AnimalEyes_Highlight_Radius, sortPriority=55}
  ) {hardnessPercent=90}
  n4 = NamedRerouteDeclaration(None=LinearInterpolate(B=@n3, Alpha=ScalarParameter() {defaultValue=1, group=AnimalEyes_Highlight, parameterName=UseAnimalEyesHighlight, sortPriority=55}) {constAlpha=0}) {name=AnimalHighlight, nodeColor=<rgba 0.7812 0 1 1>}
  n5 = LinearInterpolate(B=@n3, Alpha=ScalarParameter() {group=AnimalEyes_Highlight, parameterName=AnimalEyes_Highlight_EmissiveStrength, sortPriority=55}) {constAlpha=0}
  FunctionOutput(None=SetMaterialAttributes!(MaterialAttributes=@n1.MaterialAttributes, BaseColor=Add(A=@n1.BaseColor, B=@n4), EmissiveColor=Add(A=@n1.EmissiveColor, B=@n5)) {attributeSetTypes [#1, #2]})
  FunctionOutput(None=@n3) {outputName=AnimalHighlightMask, sortPriority=10}
 ]}
