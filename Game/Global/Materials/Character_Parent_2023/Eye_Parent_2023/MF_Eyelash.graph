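# Material function for the character eye parent material that layers eyelash shading onto incoming material attributes, behind the UseEyelashes static switch.
# A selectable vertex color channel, green by default, blends in a subsurface eyelash layer with parameterized color and surface values.
# Its opacity mask reads one SubUV cell of a mask texture.
# Its normal mixes flat tangent with a fixed local-space direction, negated on backfaces via TwoSidedSign.
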MaterialFunction /Game/Global/Materials/Character_Parent_2023/Eye_Parent_2023/MF_Eyelash
graph {expressions {
  MaterialExpressionBlendMaterialAttributes_1 {
   class=MaterialExpressionBlendMaterialAttributes
   inputs [
    A {from=MaterialExpressionNamedRerouteDeclaration_1, output=""}
    B {from=MaterialExpressionSetMaterialAttributes_1, output=""}
    Alpha {from=MaterialExpressionChannelMaskParameter_1, output=R}
   ]
   properties {}
  }
  MaterialExpressionChannelMaskParameter_1 {
   class=MaterialExpressionChannelMaskParameter
   inputs [None {from=MaterialExpressionVertexColor_1, output=""}]
   properties {group=Eyelashes, maskChannel=Green, parameterName=EyelashVertexColorMaskChannel, sortPriority=10, defaultValue=<rgba 0 1 0 0>}
  }
  MaterialExpressionComponentMask_0 {class=MaterialExpressionComponentMask, inputs [None {from=MaterialExpressionMaterialFunctionCall_0, output=RGB}], properties {r=true}}
  MaterialExpressionConvert_0 {
   class=MaterialExpressionConvert
   inputs [
    None {from=""}
    None {from=""}
    None {from=""}
   ]
   properties {
    nodeName=Make Vector3
    convertInputs [
     {type=Scalar, expressionInput {expression=null, inputName=None, mask=0, maskA=0, maskB=0, maskG=0, maskR=0, outputIndex=0}}
     {type=Scalar, expressionInput {expression=null, inputName=None, mask=0, maskA=0, maskB=0, maskG=0, maskR=0, outputIndex=0}}
     {
      type=Scalar
      defaultValue=<rgba 1 0 0 1>
      expressionInput {expression=null, inputName=None, mask=0, maskA=0, maskB=0, maskG=0, maskR=0, outputIndex=0}
     }
    ]
    convertOutputs [{type=Vector3}]
   }
  }
  MaterialExpressionConvert_1 {
   class=MaterialExpressionConvert
   inputs [
    None {from=MaterialExpressionScalarParameter_9}
    None {from=MaterialExpressionScalarParameter_10}
   ]
   properties {
    nodeName=Make Float2
    convertInputs [
     {
      type=Scalar
      expressionInput {
       expression=&self:MaterialExpressionScalarParameter_9
       inputName=None
       mask=0
       maskA=0
       maskB=0
       maskG=0
       maskR=0
       outputIndex=0
      }
     }
     {
      type=Scalar
      expressionInput {
       expression=&self:MaterialExpressionScalarParameter_10
       inputName=None
       mask=0
       maskA=0
       maskB=0
       maskG=0
       maskR=0
       outputIndex=0
      }
     }
    ]
    convertOutputs [{type=Vector2}]
   }
  }
  MaterialExpressionConvert_2 {
   class=MaterialExpressionConvert
   inputs [
    None {from=""}
    None {from=""}
    None {from=""}
   ]
   properties {
    nodeName=Make Vector3
    convertInputs [
     {type=Scalar, expressionInput {expression=null, inputName=None, mask=0, maskA=0, maskB=0, maskG=0, maskR=0, outputIndex=0}}
     {
      type=Scalar
      defaultValue=<rgba 1 0 0 1>
      expressionInput {expression=null, inputName=None, mask=0, maskA=0, maskB=0, maskG=0, maskR=0, outputIndex=0}
     }
     {
      type=Scalar
      defaultValue=<rgba 0.1 0 0 1>
      expressionInput {expression=null, inputName=None, mask=0, maskA=0, maskB=0, maskG=0, maskR=0, outputIndex=0}
     }
    ]
    convertOutputs [{type=Vector3}]
   }
  }
  MaterialExpressionFunctionInput_1 {
   class=MaterialExpressionFunctionInput
   inputs [Preview {from="", output=""}]
   properties {inputName=Material, inputType=FunctionInput_MaterialAttributes}
  }
  MaterialExpressionFunctionOutput_0 {class=MaterialExpressionFunctionOutput, inputs [None {from=MaterialExpressionStaticSwitchParameter_1, output=""}]}
  MaterialExpressionLinearInterpolate_2 {
   class=MaterialExpressionLinearInterpolate
   inputs [
    A {from=MaterialExpressionLinearInterpolate_3}
    B {from=MaterialExpressionMultiply_2}
    Alpha {from=MaterialExpressionSaturate_0}
   ]
  }
  MaterialExpressionLinearInterpolate_3 {
   class=MaterialExpressionLinearInterpolate
   inputs [
    A {from=MaterialExpressionConvert_0}
    B {from=MaterialExpressionNormalize_2}
    Alpha {from=MaterialExpressionScalarParameter_1}
   ]
  }
  MaterialExpressionMaterialFunctionCall_0 {
   class=MaterialExpressionMaterialFunctionCall
   inputs [
    Texture {from=MaterialExpressionTextureObjectParameter_1}
    UVs {from=MaterialExpressionTextureCoordinate_1}
    SubImages {from=MaterialExpressionConvert_1}
    Frame {from=MaterialExpressionScalarParameter_11}
   ]
   properties {materialFunction=/Game/Weapons/WeaponSkins/MaterialFunction/CustomNumbers_SubUV}
  }
  MaterialExpressionMultiply_0 {
   class=MaterialExpressionMultiply
   inputs [
    A {from=MaterialExpressionTwoSidedSign_0}
    B {from=""}
   ]
   properties {constB=-1}
  }
  MaterialExpressionMultiply_1 {
   class=MaterialExpressionMultiply
   inputs [
    A {from=MaterialExpressionScalarParameter_6}
    B {from=MaterialExpressionVectorParameter_1, output=RGB}
   ]
  }
  MaterialExpressionMultiply_2 {
   class=MaterialExpressionMultiply
   inputs [
    A {from=MaterialExpressionLinearInterpolate_3}
    B {from=""}
   ]
   properties {constB=-1}
  }
  MaterialExpressionNamedRerouteDeclaration_1 {
   class=MaterialExpressionNamedRerouteDeclaration
   inputs [None {from=MaterialExpressionFunctionInput_1}]
   properties {name=IN_Material, nodeColor=<rgba 0.0547 0 1 1>}
  }
  MaterialExpressionNamedRerouteUsage_1 {class=MaterialExpressionNamedRerouteUsage}
  MaterialExpressionNormalize_2 {class=MaterialExpressionNormalize, inputs [VectorInput {from=MaterialExpressionTransform_0, output=""}]}
  MaterialExpressionReroute_1 {class=MaterialExpressionReroute, inputs [None {from=MaterialExpressionLinearInterpolate_2}]}
  MaterialExpressionSaturate_0 {class=MaterialExpressionSaturate, inputs [None {from=MaterialExpressionMultiply_0}]}
  MaterialExpressionScalarParameter_1 {
   class=MaterialExpressionScalarParameter
   properties {defaultValue=0.25, group=Eyelashes, parameterName=EyelashCustomNormalMix, sliderMax=1, sortPriority=25}
  }
  MaterialExpressionScalarParameter_10 {class=MaterialExpressionScalarParameter, properties {defaultValue=1, group=Eyelashes, parameterName=EyelashMaskRows, sortPriority=31}}
  MaterialExpressionScalarParameter_11 {class=MaterialExpressionScalarParameter, properties {group=Eyelashes, parameterName=EyelashMaskIndex}}
  MaterialExpressionScalarParameter_3 {class=MaterialExpressionScalarParameter, properties {group=Eyelashes, parameterName=EyelashMetallic, sortPriority=23}}
  MaterialExpressionScalarParameter_4 {class=MaterialExpressionScalarParameter, properties {defaultValue=0.5, group=Eyelashes, parameterName=EyelashSpec, sortPriority=21}}
  MaterialExpressionScalarParameter_5 {class=MaterialExpressionScalarParameter, properties {defaultValue=0.6, group=Eyelashes, parameterName=EyelashRoughness, sortPriority=22}}
  MaterialExpressionScalarParameter_6 {
   class=MaterialExpressionScalarParameter
   properties {defaultValue=0.25, group=Eyelashes, parameterName=EyelashSubsurface, sliderMax=1, sortPriority=24}
  }
  MaterialExpressionScalarParameter_9 {class=MaterialExpressionScalarParameter, properties {defaultValue=1, group=Eyelashes, parameterName=EyelashMaskColumns, sortPriority=30}}
  MaterialExpressionSetMaterialAttributes_1 {
   class=MaterialExpressionSetMaterialAttributes
   inputs [
    MaterialAttributes {from="", output=""}
    Normal {from=MaterialExpressionReroute_1, output=""}
    ShadingModel {from=MaterialExpressionShadingModel_1, output=""}
    SubsurfaceColor {from=MaterialExpressionMultiply_1, output=""}
    BaseColor {from=MaterialExpressionVectorParameter_1, output=RGB}
    Metallic {from=MaterialExpressionScalarParameter_3, output=""}
    Specular {from=MaterialExpressionScalarParameter_4, output=""}
    Roughness {from=MaterialExpressionScalarParameter_5, output=""}
    OpacityMask {from=MaterialExpressionComponentMask_0, output=""}
   ]
   properties {attributeSetTypes [
     #1
     #2
     #3
     #4
     #5
     #6
     #7
     #8
    ]}
  }
  MaterialExpressionShadingModel_1 {class=MaterialExpressionShadingModel, inputs [], properties {shadingModel=MSM_Subsurface}}
  MaterialExpressionStaticSwitchParameter_1 {
   class=MaterialExpressionStaticSwitchParameter
   inputs [
    True {from=MaterialExpressionBlendMaterialAttributes_1, output=""}
    False {from=MaterialExpressionNamedRerouteUsage_1, output=""}
   ]
   properties {group=Eyelashes, parameterName=UseEyelashes, sortPriority=0}
  }
  MaterialExpressionTextureCoordinate_1 {class=MaterialExpressionTextureCoordinate}
  MaterialExpressionTextureObjectParameter_1 {
   class=MaterialExpressionTextureObjectParameter
   inputs [
    UVs {from="", output=""}
    Apply View MipBias {from="", output=""}
   ]
   properties {
    group=Eyelashes
    parameterName=EyelashMask
    samplerType=SAMPLERTYPE_Masks
    sortPriority=11
    texture=&T_WhiteMask
   }
  }
  MaterialExpressionTransform_0 {
   class=MaterialExpressionTransform
   inputs [None {from=MaterialExpressionConvert_2}]
   properties {transformSourceType=TRANSFORMSOURCE_Local, transformType=TRANSFORM_Tangent}
  }
  MaterialExpressionTwoSidedSign_0 {class=MaterialExpressionTwoSidedSign, inputs [], properties {}}
  MaterialExpressionVectorParameter_1 {
   class=MaterialExpressionVectorParameter
   properties {group=Eyelashes, parameterName=EyelashColor, sortPriority=20, defaultValue=<rgba 0 0 0 1>}
  }
  MaterialExpressionVertexColor_1 {class=MaterialExpressionVertexColor, inputs [], properties {}}
 }}
