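# Parent material for animal eyes on skeletal, morph-target characters, built on material attributes with the shading model set by expression.
# The shared character PreFX function feeds the animal eye function, whose output passes through highlight and eyelash functions into the shared PostFX function.
# The eye function's sclera UV output connects to both highlight inputs, and PreFX's M output drives PostFX's mask, position, distance-blend and Fresnel inputs.
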Material /Game/Global/Materials/Character_Parent_2023/Eye_Parent_2023/AnimalEyes_Parent/M_AnimalEyes_Parent_2026
graph {
 properties {
  bUseMaterialAttributes=true
  bUsedWithMorphTargets=true
  bUsedWithSkeletalMesh=true
  pixelDepthOffsetMode=PDOM_Legacy
  shadingModel=MSM_FromMaterialExpression
  usedShadingModels=Default Lit | Subsurface | Clear Coat | Subsurface Profile | Cloth
 }
 nodes [
  n1 = MaterialFunctionCall() {materialFunction=&MF_CharacterParent_2023_PreFX}
  n2 = MaterialFunctionCall!(IN_Material=@n1.Result) {materialFunction=/Game/Global/Materials/Character_Parent_2023/Eye_Parent_2023/AnimalEyes_Parent/MF/MFW_AnimalEyes}
  MaterialFunctionCall(
   Material=MaterialFunctionCall!(Material=MaterialFunctionCall!(In_UV=@n2.Eye_Sclera_UV, In=@n2.Eye_Sclera_UV) {materialFunction=/Game/Global/Materials/Character_Parent_2023/Eye_Parent_2023/AnimalEyes_Parent/MF/MF_AnimalEyeHighLight}.Result) {materialFunction=/Game/Global/Materials/Character_Parent_2023/Eye_Parent_2023/MF_Eyelash}.Result
   M=@n1.M
   Pre Skinned Local Position=@n1.M
   Distance Blend=@n1.M
   Fresnel=@n1.M
  ) {materialFunction=&MF_CharacterParent_2023_PostFX}
 ]
}
