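# Parent material for anime-shaded character eyes on skeletal meshes, using material attributes and the Cloth shading model.
# MF_eyes_2023 supplies eye UVs, normals and highlights to the shared anime shader's Pre, Calculations and Post function chain.
# The flattened eye normal blends with an optional normal map before the shading calculations.
# Saturated highlights are max-blended into lit and shaded colors, with a parameter dimming them in shade.
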Material /Game/Global/Materials/Character_Parent_2023/Eye_Parent_2023/M_Eyes_AnimeShader_2026
graph {
 properties {
  bAutomaticallySetUsageInEditor=false
  bUseMaterialAttributes=true
  bUsedWithSkeletalMesh=true
  pixelDepthOffsetMode=PDOM_Legacy
  shadingModel=MSM_Cloth
  usedShadingModels=Cloth
 }
 nodes [
  n1 = MaterialFunctionCall() {materialFunction=&/MF_eyes_2023}
  n2 = MaterialFunctionCall!(CustomUV=@n1.EyeUVs) {materialFunction=/Game/Global/Materials/Character_Parent_2023/Anime_Parent_2023/MaterialFunctions/MFC_AnimeShader_2023_Pre}
  n3 = ScalarParameter() {defaultValue=0.25, group=01_Eye_Advanced, parameterName=Eye_Iris Normal Flatten, sliderMax=1, sortPriority=21}
  n4 = TextureSampleParameter2D() {
   group=01_Anime Maps
   parameterName=Normals
   samplerType=SAMPLERTYPE_Normal
   sortPriority=13
   texture=&FlatNormalBlue
  }
  n5 = Constant3Vector() {constant=<rgba 0 0 1 0>}
  n6 = MaterialFunctionCall!(
   CustomLightVector
   OverrideNormal_WS
   STT=@n2.STT
   Normal_Map=MaterialFunctionCall(
    BaseNormal=MaterialFunctionCall(Normal=@n1.EyeNormal, Flatness=@n3) {materialFunction=&FlattenNormal}.Result
    AdditionalNormal=StaticSwitchParameter(True=@n4.RGB, False=@n5) {group=01_Anime Maps, parameterName=UseNormalMap, sortPriority=3}
   ) {materialFunction=&BlendAngleCorrectedNormals}.Result
   DFL=@n2.STT
   FX=@n2.STT
  ) {materialFunction=/Game/Global/Materials/Character_Parent_2023/Anime_Parent_2023/MaterialFunctions/MFC_AnimeShader_2023_Calculations}
  n7 = Saturate(None=NamedRerouteUsage())
  n8 = ScalarParameter() {defaultValue=0.6, group=01_Eye_Advanced, parameterName=Eye Highlight Shaded Darkening, sortPriority=124}
  MaterialFunctionCall!(
   Shading_Gradient=@n6.Shading_Gradient
   Threshold_Gradient=@n6.Shading_Gradient
   Specular_Gradient=@n6.Shading_Gradient
   Sheen_Mask=@n6.Shading_Gradient
   Thickness=@n6.Shading_Gradient
   DFL_Mask=@n6.Shading_Gradient
   FX=@n6.Shading_Gradient
   Metallic_Gradient=@n6.Shading_Gradient
   RimLight_Gradient=@n6.Shading_Gradient
   NormalBuffer=@n6.Shading_Gradient
   --------
   CINE_LightSources=@n6.Shading_Gradient
   CINE_LightColor=@n6.Shading_Gradient
   ---------
   Color_Lit=Max(A=@n2.Color_Lit, B=@n7)
   Color_Shaded=Max(A=@n2.Color_Shaded, B=Multiply(A=@n7, B=@n8))
   SRM=@n2.SRM
   Emissive=@n2.SRM
   DFL_Color=@n2.SRM
   Specular_Color=@n2.SRM
   RimLight_Color=@n2.SRM
  ) {materialFunction=/Game/Global/Materials/Character_Parent_2023/Anime_Parent_2023/MaterialFunctions/MFC_AnimeShader_2023_Post}
  NamedRerouteDeclaration(None=@n1.EyeHighlights) {name=EyeHighlights, nodeColor=<rgba 0 1 0.6719 1>}
 ]
}
